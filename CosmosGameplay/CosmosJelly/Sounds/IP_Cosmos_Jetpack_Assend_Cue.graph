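# Sound cue for the Cosmos jetpack ascend sound, built by layering two one-shot waves.
# A mixer plays the pickup wave immediately alongside the jetpack start wave, which follows after a short fixed delay with a slightly boosted, randomized volume.
# The whole cue is pitched up and lowered in volume, using shared jetpack attenuation and the gadgets/items sound class.
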SoundCue /CosmosGameplay/CosmosJelly/Sounds/IP_Cosmos_Jetpack_Assend_Cue
root=Mixer_0
nodes [
 {id=Delay_0, children [Modulator_1], properties {delayMax=0.2, delayMin=0.2}}
 {id=Mixer_0, children [Delay_0,WavePlayer_1], properties {inputVolume [1,1]}}
 {id=Modulator_1, children [WavePlayer_0], properties {volumeMax=1.2, volumeMin=1.1}}
 {id=WavePlayer_0, properties {soundWaveAssetPtr=&/IP_Cosmos_Jetpack_Start01}}
 {id=WavePlayer_1, properties {soundWaveAssetPtr=&/IP_Cosmos_Jetpack_Pickup01}}
]
properties {
 attenuationSettings=/Game/Sounds/Attenuation/Fort_Jetpack_Man_Att
 maxDistance=10572.5
 pitchMultiplier=1.5
 soundClassObject=&Fort_Gadgets_Items
 volumeMultiplier=0.7
}
