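# Sound cue for dropping the Cosmos jetpack item, part of the CosmosJelly gameplay content.
# A single drop wave plays through a Modulator that holds volume fixed and leaves pitch modulation at its defaults.
# The cue lowers overall volume, raises pitch slightly, and borrows the shared weapon-drop attenuation and loot-collect sound class.
# A concurrency override caps simultaneous instances at two.
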SoundCue /CosmosGameplay/CosmosJelly/Sounds/IP_Cosmos_Jetpack_Drop_Cue
root=Modulator_1
nodes [
 {id=Modulator_1, children [WavePlayer_0], properties {volumeMax=1, volumeMin=1}}
 {id=WavePlayer_0, properties {soundWaveAssetPtr=&/IP_Cosmos_Jetpack_Drop01}}
]
properties {
 attenuationSettings=&Fort_Weapon_Drop_Att
 bOverrideConcurrency=true
 maxDistance=3000
 pitchMultiplier=1.1
 soundClassObject=&Fort_Loot_Collect
 volumeMultiplier=0.4
 concurrencyOverrides {maxCount=2}
}
