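# Sound cue for the Cosmos jetpack running out of fuel, in the gadgets and items sound class.
# A Mixer layers a fuel-out wave with a jetpack warning wave at equal volume, so both play together.
# A Modulator at the root slightly randomizes the boosted volume of the mix, while the cue's overall multiplier lowers it.
# The cue references a short jetpack attenuation asset.
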SoundCue /CosmosGameplay/CosmosJelly/Sounds/IP_Cosmos_Jetpack_FuelOut_Cue
root=Modulator_1
nodes [
 {id=Mixer_0, children [WavePlayer_1,WavePlayer_0], properties {inputVolume [1,1]}}
 {id=Modulator_1, children [Mixer_0], properties {volumeMax=1.3, volumeMin=1.2}}
 {id=WavePlayer_0, properties {soundWaveAssetPtr=&/IP_Cosmos_Jetpack_Warning06}}
 {id=WavePlayer_1, properties {soundWaveAssetPtr=&/IP_Cosmos_Jetpack_FuelOut01}}
]
properties {
 attenuationSettings=/Game/Sounds/Attenuation/Fort_Jetpack_Man_Short_Att
 maxDistance=7430.75
 soundClassObject=&Fort_Gadgets_Items
 volumeMultiplier=0.5
 attenuationOverrides {bAttenuate=false, bEnableReverbSend=false, bSpatialize=false}
}
