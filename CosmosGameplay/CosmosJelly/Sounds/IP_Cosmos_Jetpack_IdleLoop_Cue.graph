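# Sound cue for the looping idle sound of the Cosmos jetpack, played as a gadget/item sound with jetpack attenuation.
# A looping wave player feeds a continuous modulator driven by a "height" parameter.
# As height rises, pitch climbs up to double while volume eases down from a slight boost, under a reduced overall cue volume.
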SoundCue /CosmosGameplay/CosmosJelly/Sounds/IP_Cosmos_Jetpack_IdleLoop_Cue
root=ModulatorContinuous_0
nodes [
 {
  id=ModulatorContinuous_0
  children [WavePlayer_2]
  properties {
   pitchModulationParams {default=0, maxInput=2000, maxOutput=2, minOutput=1, parameterName=height}
   volumeModulationParams {default=0, maxInput=2000, minOutput=1.3, parameterName=height}
  }
 }
 {id=WavePlayer_2, properties {bLooping=true, soundWaveAssetPtr=&/IP_Cosmos_Jetpack_IdleLoop01}}
]
properties {
 attenuationSettings=/Game/Sounds/Attenuation/Fort_Jetpack_Man_Att
 maxDistance=10572.5
 soundClassObject=&Fort_Gadgets_Items
 volumeMultiplier=0.4
}
