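# Sound cue for the Cosmos jetpack pickup, classed as a loot collect sound.
# A mixer layers a Cosmos-specific pickup wave, randomly pitched slightly upward by a modulator, over the stock Athena jetpack pickup wave at a lower input level.
# It uses weapon-drop attenuation and overrides concurrency to cap simultaneous instances at a small count.
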SoundCue /CosmosGameplay/CosmosJelly/Sounds/IP_Cosmos_Jetpack_Pickup_Cue
root=Mixer_0
nodes [
 {id=Mixer_0, children [Modulator_1,WavePlayer_0], properties {inputVolume [0.7,0.5]}}
 {id=Modulator_1, children [WavePlayer_1], properties {pitchMax=1.15, pitchMin=1.05}}
 {id=WavePlayer_0, properties {soundWaveAssetPtr=/Game/Athena/Sounds/Items/Jetpack/Jetpack_Pickup_01}}
 {id=WavePlayer_1, properties {soundWaveAssetPtr=&/IP_Cosmos_Jetpack_Pickup01}}
]
properties {
 attenuationSettings=&Fort_Weapon_Drop_Att
 bOverrideConcurrency=true
 maxDistance=3000
 soundClassObject=&Fort_Loot_Collect
 volumeMultiplier=0.4
 concurrencyOverrides {maxCount=2}
}
