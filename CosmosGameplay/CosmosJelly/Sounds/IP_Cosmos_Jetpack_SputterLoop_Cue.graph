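# IP_Cosmos_Jetpack_SputterLoop_Cue is a SoundCue for the looping sputter sound of the Cosmos jetpack.
# A Mixer blends two looping wave players at equal reduced volume: the Cosmos-specific sputter loop and a shared Athena jetpack sputter layer.
# It plays in the gadgets and items sound class, with jetpack attenuation settings for distance falloff.
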SoundCue /CosmosGameplay/CosmosJelly/Sounds/IP_Cosmos_Jetpack_SputterLoop_Cue
root=Mixer_0
nodes [
 {id=Mixer_0, children [WavePlayer_2,WavePlayer_0], properties {inputVolume [0.6,0.6]}}
 {id=WavePlayer_0, properties {bLooping=true, soundWaveAssetPtr=/FNGameplayCues/Athena/Sounds/Items/Jetpack/Jetpack_SputterLoopLayer_01}}
 {id=WavePlayer_2, properties {bLooping=true, soundWaveAssetPtr=&/IP_Cosmos_Jetpack_SputterLoop01}}
]
properties {
 attenuationSettings=/Game/Sounds/Attenuation/Fort_Jetpack_Man_Att
 maxDistance=10572.5
 soundClassObject=&Fort_Gadgets_Items
 volumeMultiplier=1
}
