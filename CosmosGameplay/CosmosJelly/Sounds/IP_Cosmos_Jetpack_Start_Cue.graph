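# Sound cue that plays the start sound for a jetpack in the CosmosGameplay CosmosJelly content.
# A single wave player feeds a modulator that slightly randomizes volume, with cue-level multipliers lowering volume and nudging pitch up.
# It uses jetpack attenuation and the gadget items sound class.
# A concurrency override allows one instance per owner and blocks new plays while one is active.
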SoundCue /CosmosGameplay/CosmosJelly/Sounds/IP_Cosmos_Jetpack_Start_Cue
root=Modulator_1
nodes [
 {id=Modulator_1, children [WavePlayer_0], properties {volumeMax=1.2, volumeMin=1.1}}
 {id=WavePlayer_0, properties {soundWaveAssetPtr=&/IP_Cosmos_Jetpack_Start01}}
]
properties {
 attenuationSettings=/Game/Sounds/Attenuation/Fort_Jetpack_Man_Att
 bOverrideConcurrency=true
 maxDistance=10572.5
 pitchMultiplier=1.05
 soundClassObject=&Fort_Gadgets_Items
 volumeMultiplier=0.5
 concurrencyOverrides {bLimitToOwner=true, maxCount=1, resolutionRule=PreventNew}
}
