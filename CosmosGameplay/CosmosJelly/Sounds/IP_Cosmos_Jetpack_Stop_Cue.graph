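# Sound cue that plays the stop sound for the Cosmos jetpack in the CosmosJelly gameplay content.
# A single wave player feeds the IP_Cosmos_Jetpack_Stop02 wave into a Modulator node at the root, left at its default settings.
# It plays at reduced volume under the gadgets and items sound class, spatialized through a short jetpack attenuation asset.
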SoundCue /CosmosGameplay/CosmosJelly/Sounds/IP_Cosmos_Jetpack_Stop_Cue
root=Modulator_1
nodes [
 {id=Modulator_1, children [WavePlayer_0]}
 {id=WavePlayer_0, properties {soundWaveAssetPtr=&/IP_Cosmos_Jetpack_Stop02}}
]
properties {
 attenuationSettings=/Game/Sounds/Attenuation/Fort_Jetpack_Man_Short_Att
 maxDistance=7430.75
 soundClassObject=&Fort_Gadgets_Items
 volumeMultiplier=0.7
}
