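# Sound cue for the Cosmos jetpack warning sound.
# A Mixer layers two warning waves, playing them together at reduced levels with Warning08 louder than Warning04.
# It uses the shared Fort jetpack attenuation and the gadget items sound class.
# A concurrency override caps it at two simultaneous instances, blocking new plays beyond that.
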SoundCue /CosmosGameplay/CosmosJelly/Sounds/IP_Cosmos_Jetpack_Warning_Cue
root=Mixer_0
nodes [
 {id=Mixer_0, children [WavePlayer_2,WavePlayer_3], properties {inputVolume [0.3,0.6]}}
 {id=WavePlayer_2, properties {soundWaveAssetPtr=&/IP_Cosmos_Jetpack_Warning04}}
 {id=WavePlayer_3, properties {soundWaveAssetPtr=&/IP_Cosmos_Jetpack_Warning08}}
]
properties {
 attenuationSettings=/Game/Sounds/Attenuation/Fort_Jetpack_Att
 maxDistance=13067
 soundClassObject=&Fort_Gadgets_Items
 volumeMultiplier=0.8
 concurrencyOverrides {maxCount=2, resolutionRule=PreventNew}
}
