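# Mono MetaSound for the CosmosJelly jetpack, mixing engine, ascend and fuel-sputter layers, each with a randomly chosen wave and its own ADSR envelope.
# Engine start triggers engine and sputter; ascend adds its layer from a random start time and ducks the engine.
# A fuel-driven wavetable curve crossfades sputter against the other two layers.
# Vertical movement, smoothed object proximity and random offsets shift each layer's pitch.
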MetaSoundSource /CosmosGameplay/CosmosJelly/Sounds/IP_MSS_Jethro_CosmosJelly_Jetpack
inputs [
 Fuel.Normalized {type=Float}
 Interp Time {type=Time}
 Movement.Speed {type=Float}
 Movement.SpeedAzimuth {type=Float}
 Movement.SpeedElevation {type=Float}
 NormalizedDistanceToObjects {type=Float}
 Play.Ascend {type=Trigger}
 Play.StartEngine {type=Trigger}
 Play.StopAscend {type=Trigger}
 Play.StopEngine {type=Trigger}
 UE.Source.OnPlay {type=Trigger}
]
interfaces [Fort.Movement:1.0, UE.OutputFormat.Mono:1.0, UE.Source.OneShot:1.0, UE.Source:1.1]
outputs [
 "UE.OutputFormat.Mono.Audio:0" {type=Audio}
 UE.Source.OneShot.OnFinished {type=Trigger}
]
pages [Default {
  input_defaults {
   Fuel.Normalized=0
   Interp Time=0.3
   Movement.Speed=0
   Movement.SpeedAzimuth=0
   Movement.SpeedElevation=0
   NormalizedDistanceToObjects=0
   Play.Ascend=null
   Play.StartEngine=null
   Play.StopAscend=null
   Play.StopEngine=null
   UE.Source.OnPlay=null
  }
  nodes [
   DecibelstoLinearGain_0 = UE.Decibels to Linear Gain:Float() {also {defaults {Decibels=-8.78}}}
   DecibelstoLinearGain_2 = UE.Decibels to Linear Gain:Float() {also {defaults {Decibels=-7.95}}}
   n1 = InitVariable.Audio()
   ADSREnvelopeADSREnvelope_0 = ADSR Envelope.ADSR Envelope:Audio(Trigger Attack=^Play.StartEngine, Trigger Release=^Play.StopEngine) {also {defaults {Sustain Level=1}}}
   n2 = Clamp.Clamp:Float(In=UE.WaveTableBankEvaluate(Input=^Fuel.Normalized) {also {defaults {WaveTableBank=&/IP_WTB_Jethro_CosmosJelly_JetpackSputterCurve}}}.Output) {also {defaults {Max=1, Min=0}}}
   n3 = VariableMutator.Float(Variable=InitVariable.Float().Variable, Value=UE.Subtract:Float(AdditionalOperands=@n2.Value) {also {defaults {PrimaryOperand=1}}}.Out)
   n4 = VariableAccessor.Float(Variable=VariableMutator.Float(
     Variable=InitVariable.Float().Variable
     Value=MapRange.MapRange:Float(In=^Movement.SpeedElevation) {also {defaults {In Range A=-1}}}.Out Value
    ).Variable)
   n5 = VariableAccessor.Float(Variable=@n4.Variable)
   n6 = MapRange.MapRange:Float(In=VariableAccessor.Float(Variable=@n5.Variable).Value) {also {defaults {Out Range B=5}}}
   n7 = Clamp.Clamp:Float(In=UE.Subtract:Float(AdditionalOperands=^NormalizedDistanceToObjects) {also {defaults {PrimaryOperand=1}}}.Out) {also {defaults {Max=1, Min=0}}}
   n8 = VariableAccessor.Float(Variable=VariableMutator.Float(Variable=InitVariable.Float().Variable, Value=UE.InterpTo:Audio(Interp Time=^Interp Time, Target=@n7.Value).Value).Variable)
   n9 = VariableAccessor.Float(Variable=@n8.Variable)
   n10 = MapRange.MapRange:Float(In=VariableAccessor.Float(Variable=@n9.Variable).Value) {also {defaults {Out Range B=12}}}
   ArrayRandomGet_2 = Array.Random Get:WaveAsset:Array(Next=^Play.StartEngine, In Array=VariableAccessor.WaveAsset:Array(Variable=InitVariable.WaveAsset:Array().Variable).Value) {also {defaults {Weights []}}}
   n11 = UE.RandomFloat(Next=@ArrayRandomGet_2.On Next, Reset=@ArrayRandomGet_2.On Reset)
   n12 = VariableAccessor.Trigger(Variable=VariableMutator.Trigger(Value=@ADSREnvelopeADSREnvelope_0.On Done, Variable=InitVariable.Trigger().Variable).Variable)
   WavePlayer_0 = UE.Wave Player:Mono(
    Pitch Shift=UE.Add:Float(AdditionalOperands=UE.Add:Float(AdditionalOperands=@n6.Out Value, PrimaryOperand=@n10.Out Value).Out, PrimaryOperand=@n11.Value).Out
    Wave Asset=@ArrayRandomGet_2.Value
    Play=@n11.On Next
    Stop=VariableAccessor.Trigger(Variable=@n12.Variable).Value
   ) {also {defaults {Loop="<MetaSoundBuilderResult.SUCCEEDED: 0>"}, version [1,1]}}
   n13 = UE.Multiply:Audio(
    PrimaryOperand=@ADSREnvelopeADSREnvelope_0.Out Envelope
    AdditionalOperands=UE.Multiply:Audio by Float(AdditionalOperands=VariableAccessor.Float(Variable=@n3.Variable).Value, PrimaryOperand=@WavePlayer_0.Out Mono).Out
   )
   n14 = InitVariable.Audio()
   ADSREnvelopeADSREnvelope_2 = ADSR Envelope.ADSR Envelope:Audio(Trigger Attack=^Play.StartEngine, Trigger Release=^Play.StopEngine)
   DecibelstoLinearGain_1 = UE.Decibels to Linear Gain:Float() {also {defaults {Decibels=-6}}}
   n15 = InitVariable.Trigger()
   TriggerOnce_1 = UE.Trigger Once(Reset=^Play.Ascend, Trigger In=^Play.StopAscend)
   TriggerOnce_0 = UE.Trigger Once(Trigger In=^Play.Ascend, Reset=^Play.StopAscend)
   n16 = VariableAccessor.Trigger(Variable=VariableMutator.Trigger(Variable=InitVariable.Trigger().Variable, Value=@TriggerOnce_0.Trigger Out).Variable)
   TriggerRouteTriggerRouteFloat2_0 = "TriggerRoute.Trigger Route (Float, 2):Float"(
    Value 0=@DecibelstoLinearGain_1.Linear Gain
    Set 1=VariableAccessor.Trigger(Variable=VariableMutator.Trigger(Variable=@n15.Variable, Value=@TriggerOnce_1.Trigger Out).Variable).Value
    Set 0=@n16.Value
   ) {also {defaults {Value 1=1}}}
   n17 = UE.InterpTo:Audio(Target=@TriggerRouteTriggerRouteFloat2_0.Value) {also {defaults {Interp Time=0.5}}}
   n18 = InitVariable.Float()
   n19 = MapRange.MapRange:Float(In=@n8.Value) {also {defaults {Out Range B=12}}}
   n20 = MapRange.MapRange:Float(In=@n4.Value) {also {defaults {Out Range B=5}}}
   ArrayRandomGet_1 = Array.Random Get:WaveAsset:Array(Next=^Play.StartEngine, In Array=VariableAccessor.WaveAsset:Array(Variable=InitVariable.WaveAsset:Array().Variable).Value) {also {defaults {Weights []}}}
   n21 = UE.RandomFloat(Next=@ArrayRandomGet_1.On Next, Reset=@ArrayRandomGet_1.On Reset)
   n22 = VariableAccessor.Trigger(Variable=VariableMutator.Trigger(Value=@ADSREnvelopeADSREnvelope_2.On Done, Variable=InitVariable.Trigger().Variable).Variable)
   WavePlayer_1 = UE.Wave Player:Mono(
    Pitch Shift=UE.Add:Float(AdditionalOperands=UE.Add:Float(PrimaryOperand=@n19.Out Value, AdditionalOperands=@n20.Out Value).Out, PrimaryOperand=@n21.Value).Out
    Wave Asset=@ArrayRandomGet_1.Value
    Play=@n21.On Next
    Stop=VariableAccessor.Trigger(Variable=@n22.Variable).Value
   ) {also {defaults {Loop="<MetaSoundBuilderResult.SUCCEEDED: 0>"}, version [1,1]}}
   n23 = UE.Multiply:Audio by Float(
    AdditionalOperands=UE.Subtract:Float(AdditionalOperands=VariableAccessor.Float(Variable=@n18.Variable).Value) {also {defaults {PrimaryOperand=1}}}.Out
    PrimaryOperand=@WavePlayer_1.Out Mono
   )
   n24 = VariableAccessor.Float(Variable=VariableMutator.Float(Value=@n2.Value, Variable=InitVariable.Float().Variable).Variable)
   n25 = UE.Multiply:Audio by Float(
    PrimaryOperand=UE.Multiply:Audio(
     PrimaryOperand=@ADSREnvelopeADSREnvelope_2.Out Envelope
     AdditionalOperands=UE.Multiply:Audio by Float(AdditionalOperands=@n17.Value, PrimaryOperand=@n23.Out).Out
    ).Out
    AdditionalOperands=@n24.Value
   )
   n26 = InitVariable.Audio()
   n27 = VariableAccessor.Trigger(Variable=@n16.Variable)
   ADSREnvelopeADSREnvelope_1 = ADSR Envelope.ADSR Envelope:Audio(Trigger Release=^Play.StopEngine, Trigger Attack=VariableAccessor.Trigger(Variable=@n27.Variable).Value) {also {defaults {Sustain Level=1}}}
   n28 = MapRange.MapRange:Float(In=@n5.Value) {also {defaults {Out Range B=3}}}
   n29 = MapRange.MapRange:Float(In=@n9.Value) {also {defaults {Out Range B=5}}}
   n30 = UE.RandomTime(Next=@n27.Value) {also {defaults {Max=3}}}
   ArrayRandomGet_0 = Array.Random Get:WaveAsset:Array(Next=@n30.On Next, In Array=VariableAccessor.WaveAsset:Array(Variable=InitVariable.WaveAsset:Array().Variable).Value) {also {defaults {Weights []}}}
   n31 = UE.RandomFloat(Next=@ArrayRandomGet_0.On Next, Reset=@ArrayRandomGet_0.On Reset)
   n32 = VariableAccessor.Trigger(Variable=VariableMutator.Trigger(Value=@ADSREnvelopeADSREnvelope_1.On Done, Variable=InitVariable.Trigger().Variable).Variable)
   WavePlayer_2 = UE.Wave Player:Mono(
    Pitch Shift=UE.Add:Float(AdditionalOperands=UE.Add:Float(AdditionalOperands=@n28.Out Value, PrimaryOperand=@n29.Out Value).Out, PrimaryOperand=@n31.Value).Out
    Wave Asset=@ArrayRandomGet_0.Value
    Play=@n31.On Next
    Start Time=@n30.Value
    Stop=VariableAccessor.Trigger(Variable=@n32.Variable).Value
   ) {also {defaults {Loop="<MetaSoundBuilderResult.SUCCEEDED: 0>"}, version [1,1]}}
   n33 = UE.Multiply:Audio by Float(
    PrimaryOperand=UE.Multiply:Audio(PrimaryOperand=@ADSREnvelopeADSREnvelope_1.Out Envelope, AdditionalOperands=@WavePlayer_2.Out Mono).Out
    AdditionalOperands=VariableAccessor.Float(Variable=@n24.Variable).Value
   )
   AudioMixerAudioMixerMono6_0 = "AudioMixer.Audio Mixer (Mono, 6)"(
    Gain 2=@DecibelstoLinearGain_0.Linear Gain
    Gain 3=@DecibelstoLinearGain_2.Linear Gain
    In 3=VariableAccessor.Audio(Variable=VariableMutator.Audio(Variable=@n1.Variable, Value=@n13.Out).Variable).Value
    In 1=VariableAccessor.Audio(Variable=VariableMutator.Audio(Variable=@n14.Variable, Value=@n25.Out).Variable).Value
    In 2=VariableAccessor.Audio(Variable=VariableMutator.Audio(Variable=@n26.Variable, Value=@n33.Out).Variable).Value
   )
   InitVariable.Trigger()
   TriggerAnyTriggerAny3_0 = "TriggerAny.Trigger Any (3)"(In 2=@n12.Value, In 0=@n32.Value, In 1=@n22.Value)
  ]
  links [
   {dst="@UE.OutputFormat.Mono.Audio:0", src=AudioMixerAudioMixerMono6_0.Out}
   {dst="@UE.Source.OneShot.OnFinished", src=TriggerAnyTriggerAny3_0.Out}
  ]
 }]
properties {
 assetImportData=&self:AssetImportData
 attenuationSettings=&/IP_Cosmos_Jetpack_ATTN
 bDebug=true
 bOverrideConcurrency=true
 maxDistance=13067
 sampleRate=48000
 soundClassObject=&Fort_Gadgets_Items
 volume=1.1885
 concurrencyOverrides {bLimitToOwner=true, maxCount=1}
}
